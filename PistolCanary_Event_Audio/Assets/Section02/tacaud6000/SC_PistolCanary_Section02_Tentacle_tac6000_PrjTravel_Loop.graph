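# Sound cue for the looping travel sound of the tentacle 6000 attack projectile in Section 02 of the PistolCanary event.
# A Random node picks one of three equally weighted looping wave players, reusing the tentacle 4000 projectile loop recordings at a lowered pitch.
# It routes to the Fort_Events sound class, uses the ground beam attenuation, and is limited by the shared projectile concurrency settings.
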SoundCue /PistolCanary_Event_Audio/Assets/Section02/tacaud6000/SC_PistolCanary_Section02_Tentacle_tac6000_PrjTravel_Loop
root=Random_0
nodes [
 {id=Random_0, children [WavePlayer_0,WavePlayer_2,WavePlayer_3], properties {weights [1,1,1]}}
 {
  id=WavePlayer_0
  properties {bLooping=true, soundWaveAssetPtr=/PistolCanary_Event_Audio/Assets/Section02/wav/PistolCanary_Tentacle_Attk_4000_Proj_Loop_01}
 }
 {
  id=WavePlayer_2
  properties {bLooping=true, soundWaveAssetPtr=/PistolCanary_Event_Audio/Assets/Section02/wav/PistolCanary_Tentacle_Attk_4000_Proj_Loop_02}
 }
 {
  id=WavePlayer_3
  properties {bLooping=true, soundWaveAssetPtr=/PistolCanary_Event_Audio/Assets/Section02/wav/PistolCanary_Tentacle_Attk_4000_Proj_Loop_03}
 }
]
properties {
 attenuationSettings=/PistolCanary_Audio/Assets/Settings/ATT_PistolCanary_Beam_Ground
 maxDistance=22000
 pitchMultiplier=0.8
 soundClassObject=&Fort_Events
 volumeMultiplier=1
 concurrencySet [/PistolCanary_Event_Audio/Assets/Settings/CONC_PistolCanary_Projectiles]
}
