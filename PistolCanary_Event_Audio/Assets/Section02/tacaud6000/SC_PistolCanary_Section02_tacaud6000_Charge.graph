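# Sound cue for the charge sound of a tentacle attack in Section 02 of the PistolCanary event.
# A single wave player plays one fixed charge wave, with no randomization or mixing, at reduced volume in the Fort_Events sound class.
# It uses shared tentacle tip attenuation and overrides concurrency to cap simultaneous instances and set a retrigger time.
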SoundCue /PistolCanary_Event_Audio/Assets/Section02/tacaud6000/SC_PistolCanary_Section02_tacaud6000_Charge
root=WavePlayer_1
nodes [{id=WavePlayer_1, properties {soundWaveAssetPtr=/PistolCanary_Event_Audio/Assets/Section02/wav/PistolCanary_Tentacle_Attk_6000_Charge_01}}]
properties {
 attenuationSettings=/PistolCanary_Audio/Assets/Settings/ATT_PistolCanary_Tentacle_Tip
 bOverrideConcurrency=true
 maxDistance=100000
 soundClassObject=&Fort_Events
 volumeMultiplier=0.7
 concurrencyOverrides {maxCount=2, retriggerTime=0.5}
}
