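# Sound cue for the orb blast of a tentacle attack in Section 02 of the PistolCanary event.
# A random node picks between two equally weighted blast wave variations.
# It plays slightly boosted through the Fort_Events sound class, with a dedicated orb blast attenuation asset for long-range audibility.
# A concurrency override caps simultaneous instances at a low count and sets a short retrigger time.
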SoundCue /PistolCanary_Event_Audio/Assets/Section02/tacaud6000/SC_PistolCanary_Section02_tacaud6000_Orb_Shoot
root=Random_0
nodes [
 {id=Random_0, children [WavePlayer_1,WavePlayer_2], properties {weights [1,1]}}
 {id=WavePlayer_1, properties {soundWaveAssetPtr=/PistolCanary_Event_Audio/Assets/Section02/wav/PistolCanary_Tentacle_Attk_6000_Orb_Blast_01}}
 {id=WavePlayer_2, properties {soundWaveAssetPtr=/PistolCanary_Event_Audio/Assets/Section02/wav/PistolCanary_Tentacle_Attk_6000_Orb_Blast_02}}
]
properties {
 attenuationSettings=/PistolCanary_Audio/Assets/Settings/ATT_PistolCanary_Tentacle_Orb_Blast
 bOverrideConcurrency=true
 maxDistance=35500
 soundClassObject=&Fort_Events
 volumeMultiplier=1.2
 concurrencyOverrides {maxCount=2, retriggerTime=0.3}
}
